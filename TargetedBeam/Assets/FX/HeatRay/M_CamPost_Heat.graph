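# Unlit alpha-composite Niagara sprite material for a screen-space heat overlay with 2D-offset refraction, in the HeatRay effects.
# Radial coordinates from viewport UV drive a panned, swizzled wave-flame texture, which is multiplied by a mask to form alpha and distortion terms.
# A power-curve screen-edge mask and smoothstep dissolve are published through named reroutes.
# Particle color, a two-color lerp and dynamic parameters control tint, emissive, time, dissolve and refraction strength.
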Material /TargetedBeam/Assets/FX/HeatRay/M_CamPost_Heat
graph {
 properties {
  bAutomaticallySetUsageInEditor=false
  bUsedWithNiagaraSprites=true
  blendMode=BLEND_AlphaComposite
  refractionMethod=RM_2DOffset
  shadingModel=MSM_Unlit
  usedShadingModels=Unlit
 }
 nodes [
  n1 = ScreenPosition()
  LinearInterpolate(A=@n1.ViewportUV, B=TextureCoordinate() {coordinateIndex=1}, Alpha=ScalarParameter() {parameterName=CubeMap_Switch})
  n2 = NamedRerouteUsage()
  n3 = ScreenPosition()
  n4 = VectorParameter() {parameterName=TextureUV, defaultValue=<rgba 1 1 1 0>}
  n5 = MaterialFunctionCall!(
   InUV=MaterialFunctionCall(Vector or UVs=Subtract(A=@n3.ViewportUV) {constB=0.5}) {materialFunction=&VectorToRadialValue}.Radial Coordinates
   Texture
   InRGB=@n4.RGB
   InA=@n4.RGB
   InTime=DynamicParameter() {defaultValue=<rgba 1 1 1 250>, paramNames [Time, DIssolveOffset, Distort, Ref]}.Time
  ) {materialFunction=/Game/Effects/Fort_Effects/Materials/_Functions/MF_TexturePanner}
  n6 = Multiply(
   A=@n2
   B=TextureSample!(
    UVs=MaterialFunctionCall(XY=@n5.UV) {materialFunction=&Swizzle}.YX
    Tex
    Level
    Apply View MipBias
   ) {mipValueMode=TMVM_MipLevel, texture=/Game/Effects/Fort_Effects/Textures/Tile/T_Tile_WaveFlame}.G
  )
  Multiply(A=@n6, B=DynamicParameter() {defaultValue=<rgba 1 1 1 250>, paramNames [Time, DIssolveOffset, Distort, Ref]}.Ref)
  n7 = ParticleColor()
  n8 = Multiply(A=@n7.RGB, B=NamedRerouteUsage())
  n9 = LinearInterpolate(
   A=VectorParameter() {parameterName=ColorB}.RGB
   B=VectorParameter() {parameterName=ColorA}.RGB
   Alpha=Multiply(A=NamedRerouteUsage(), B=@n7.A)
  )
  Multiply(A=MaterialFunctionCall(Emissive=Multiply(A=@n8, B=@n9)) {materialFunction=&MF_FX_EmissiveAdjustments}.Result) {constB=0.5}
  n10 = NamedRerouteUsage()
  n11 = DynamicParameter() {defaultValue=<rgba 1 1 1 250>, paramNames [Time, DIssolveOffset, Distort, Ref]}
  n12 = Multiply(
   A=ScalarParameter() {defaultValue=0.7595, group=Dissolve, parameterName=DissolveMin}
   B=OneMinus(None=Multiply(A=@n10, B=@n11.DIssolveOffset))
  )
  n13 = ScalarParameter() {defaultValue=4.1085, group=Dissolve, parameterName=DissolveMax}
  NamedRerouteDeclaration(None=SmoothStep(Min=@n12, Max=Add(A=@n12, B=@n13), Value=@n10)) {name=DissolveResault, nodeColor=<rgba 1 0.9405 0.1679 1>}
  NamedRerouteDeclaration(None=Add(A=@n6, B=@n2)) {name=Alpha, nodeColor=<rgba 1 0.1875 0 1>}
  n14 = Abs(None=Subtract(A=Multiply(A=@n1.ViewportUV) {constB=2}))
  n15 = ScalarParameter() {defaultValue=3, parameterName=MaskExp}
  n16 = MaterialFunctionCall(Float2=OneMinus(None=Power(Base=@n14, Exp=@n15))) {materialFunction=&BreakOutFloat2Components}
  n17 = Multiply(A=@n16.R, B=@n16.R)
  n18 = ScalarParameter() {defaultValue=0.5, parameterName=CubeMap_Select}
  NamedRerouteDeclaration(None=OneMinus(None=MaterialFunctionCall(A=@n16.R, B=@n17, C=@n16.R, Alpha=@n18) {materialFunction=&Lerp_3Color}.Result)) {name=CubeMap_2, nodeColor=<rgba 1 1 1 1>}
  NamedRerouteUsage()
  ScalarParameter() {defaultValue=0.1, parameterName=Ref}
  TextureCoordinate()
 ]
}
