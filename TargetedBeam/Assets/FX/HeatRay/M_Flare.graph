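# Unlit, two-sided, alpha-composite Niagara sprite material for a lens-flare style glow in the HeatRay beam effect.
# A procedural inverse falloff, driven by the Emi dynamic parameter and shaped by glow parameters, is clamped and tinted by particle color.
# Depth fade, camera-distance fade and sprite-edge masks from remapped UVs attenuate emissive and alpha.
# An emissive adjustment function processes the color, and a CameraOffset function provides a camera-relative position offset.
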Material /TargetedBeam/Assets/FX/HeatRay/M_Flare
graph {
 properties {
  bAutomaticallySetUsageInEditor=false
  bUsedWithNiagaraSprites=true
  blendMode=BLEND_AlphaComposite
  shadingModel=MSM_Unlit
  twoSided=true
  usedShadingModels=Unlit
 }
 nodes [
  Constant() {r=1}
  n1 = ParticleColor()
  n2 = OneMinus(None=NamedRerouteUsage())
  n3 = DynamicParameter() {paramNames [Emi,Param2,Param3,Param4]}
  n4 = ScalarParameter() {defaultValue=0.1386, parameterName=GlowBlance}
  n5 = Divide(A=Subtract(A=Divide(A=@n3.Emi, B=@n2), B=@n3.Emi), B=@n4)
  n6 = ScalarParameter() {defaultValue=0.5387, parameterName=GlowExp}
  n7 = Min(A=Multiply(A=Power(Base=@n5, Exp=@n6), B=@n4), B=ScalarParameter() {defaultValue=10, parameterName=MaxiumBright}) {constB=1000}
  n8 = Multiply(A=@n1.RGB, B=@n7)
  n9 = Multiply(A=Multiply(A=NamedRerouteUsage(), B=NamedRerouteUsage()), B=NamedRerouteUsage())
  Divide(A=MaterialFunctionCall(Emissive=Multiply(A=@n8, B=@n9)) {materialFunction=&MF_FX_EmissiveAdjustments}.Result)
  MaterialFunctionCall(Offset Amount=ScalarParameter() {defaultValue=10, parameterName=CamOffset}) {materialFunction=/Engine/Functions/Engine_MaterialFunctions02/WorldPositionOffset/CameraOffset}
  n10 = TextureCoordinate()
  n11 = TextureCoordinate() {coordinateIndex=1}
  n12 = ScalarParameter() {parameterName=CubeMap_Switch}
  n13 = Subtract(A=Multiply(A=LinearInterpolate(A=@n10, B=@n11, Alpha=@n12)) {constB=2})
  Multiply(A=@n13, B=@n13) {constB=2}
  n14 = MaterialFunctionCall(Float2=OneMinus(None=Abs(None=@n13))) {materialFunction=&BreakOutFloat2Components}
  n15 = OneMinus(None=@n14.R)
  n16 = ScalarParameter() {defaultValue=150, parameterName=CleanEnds}
  Multiply(A=OneMinus(None=Power(Base=@n15, Exp=@n16) {constExponent=150}), B=@n14.G)
  Multiply(A=Multiply(A=@n1.A, B=@n7), B=@n9)
  n17 = OneMinus(None=@n14.G)
  Multiply(A=OneMinus(None=Power(Base=@n17, Exp=@n16) {constExponent=150}), B=@n14.R)
  NamedRerouteDeclaration(None=DepthFade(FadeDistance=ScalarParameter() {defaultValue=20, parameterName=DF_FD})) {name=DepthFade, nodeColor=<rgba 1 1 1 1>}
  n18 = MaterialFunctionCall(Float2=ParticleSize()) {materialFunction=&BreakOutFloat2Components}
  n19 = Multiply(A=@n14.G, B=Divide(A=@n18.G, B=@n18.G))
  n20 = ScalarParameter() {parameterName=Sprite_EndpointExcessive}
  NamedRerouteDeclaration(None=Saturate(None=Divide(A=@n19, B=@n20))) {name=Sprite_EndpointExcessive, nodeColor=<rgba 1 0 0.806 1>}
  n21 = Subtract(A=CameraPositionWS(), B=WorldPosition().XYZ)
  n22 = Length(None=@n21) {is=class MaterialExpressionLength}
  n23 = DotProduct(A=Normalize(VectorInput=@n21), B=Transform(None=Constant3Vector() {constant=<rgba 0 0 -1 0>}) {transformSourceType=TRANSFORMSOURCE_View})
  n24 = ScalarParameter() {defaultValue=20, parameterName=Camfade_Offset}
  n25 = ScalarParameter() {defaultValue=100, parameterName=Camfade_Lemgth}
  n26 = Divide(A=Subtract(A=Multiply(A=@n22, B=@n23), B=@n24), B=@n25)
  n27 = Divide(A=Subtract(A=PixelDepth().R, B=@n24), B=@n25)
  n28 = StaticBool()
  NamedRerouteDeclaration(None=Saturate(None=StaticSwitch(True=@n26, False=@n27, Value=@n28))) {name=CamFade, nodeColor=<rgba 1 0 0.9613 1>}
  NamedRerouteDeclaration(None=MaterialFunctionCall(A=@n14.R, B=Multiply(A=@n14.R, B=@n14.R), C=@n14.R, Alpha=ScalarParameter() {defaultValue=0.5, parameterName=CubeMap_Select}) {materialFunction=&Lerp_3Color}.Result) {name=CubeMap_2, nodeColor=<rgba 1 1 1 1>}
 ]
}
