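# Unlit Modulate-blend material for Niagara sprites in the HeatRay effects, tinting what lies behind each particle.
# A radial SphereMask drives an inverse falloff that spikes at the center and reaches zero at the edge, scaled by the Emi dynamic parameter.
# That falloff, particle alpha and a depth fade lerp from neutral white to particle color.
# A CameraOffset function with a CamOffset parameter provides camera-relative sprite offset.
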Material /TargetedBeam/Assets/FX/HeatRay/M_Mudulate
graph {
 properties {
  bAutomaticallySetUsageInEditor=false
  bUsedWithNiagaraSprites=true
  blendMode=BLEND_Modulate
  shadingModel=MSM_Unlit
  usedShadingModels=Unlit
 }
 nodes [
  n1 = ParticleColor()
  n2 = Constant() {r=0.5}
  n3 = SphereMask(A=TextureCoordinate(), B=@n2, Radius=@n2) {hardnessPercent=0}
  n4 = DynamicParameter() {paramNames [Emi,Param2,Param3,Param4]}
  n5 = Subtract(A=Divide(A=@n4.Emi, B=OneMinus(None=@n3)), B=@n4.Emi)
  n6 = NamedRerouteUsage()
  LinearInterpolate(B=@n1.RGB, Alpha=Multiply(A=@n1.A, B=Multiply(A=@n5, B=@n6))) {constA=1}
  MaterialFunctionCall(Offset Amount=ScalarParameter() {defaultValue=10, parameterName=CamOffset}) {materialFunction=/Engine/Functions/Engine_MaterialFunctions02/WorldPositionOffset/CameraOffset}
  NamedRerouteDeclaration(None=DepthFade(FadeDistance=ScalarParameter() {defaultValue=20, parameterName=DF_FD})) {name=DepthFade, nodeColor=<rgba 1 1 1 1>}
 ]
}
