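# Translucent Niagara sprite material in the TargetedBeam HeatRay FX set, drawing a soft, noise-dissolved particle with responsive AA.
# A squared UV falloff is modulated by panned Perlin noise, then smoothstepped using dynamic parameters for time, dissolve and detail scale.
# The shape blends ParticleColorB toward particle color, tinted by time of day through emissive adjustments.
# The dissolve mask is also multiplied by particle alpha.
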Material /TargetedBeam/Assets/FX/HeatRay/M_SimpleParticle
graph {
 properties {
  bAutomaticallySetUsageInEditor=false
  bEnableResponsiveAA=true
  bUsedWithNiagaraSprites=true
  blendMode=BLEND_Translucent
  usedShadingModels=Default Lit
 }
 nodes [
  n1 = VectorParameter() {parameterName=ParticleColorB}
  n2 = ParticleColor()
  n3 = Subtract(A=TextureCoordinate() {uTiling=2, vTiling=2})
  n4 = MaterialFunctionCall(Float2=OneMinus(None=Abs(None=@n3))) {materialFunction=&BreakOutFloat2Components}
  n5 = Multiply(A=@n4.R, B=@n4.R)
  n6 = TextureCoordinate()
  n7 = VectorParameter() {parameterName=NoiseControl, defaultValue=<rgba 0.5 0.5 1 0.3286>}
  n8 = DynamicParameter() {defaultValue=<rgba 1 0.5 0.1 1>, paramNames [Time, DissolveOffset, DissolveEdge, DetailScale]}
  n9 = TextureSample(UVs=Multiply(
    A=MaterialFunctionCall!(InUV=@n6, Texture, InRGB=@n7.RGB, InA=@n7.RGB, InTime=@n8.Time) {materialFunction=/Game/Effects/Fort_Effects/Materials/_Functions/MF_TexturePanner}.UV
    B=@n8.DetailScale
   )) {texture=/Game/Effects/Fort_Effects/Textures/Tile/Tile_T_PerlinNoiseBlend}
  n10 = MaterialFunctionCall(A=@n9.R, B=@n9.R, C=@n9.R, Alpha=ScalarParameter() {defaultValue=0.5, parameterName=ParticleDetailType}) {materialFunction=&Lerp_3Color}
  n11 = Add(A=@n5, B=Multiply(A=Multiply(A=@n10.Result) {constB=2}, B=@n5))
  n12 = SmoothStep(Min=@n8.DissolveOffset, Max=Add(A=@n8.DissolveOffset, B=@n8.DissolveOffset), Value=@n11)
  n13 = Min(A=@n11, B=@n12)
  n14 = VectorParameter() {parameterName=ColorTint_Morning, defaultValue=<rgba 0.7969 0.5704 0.4337 1>}
  n15 = VectorParameter() {parameterName=ColorTint_Day, defaultValue=<rgba 1 1 1 0>}
  n16 = VectorParameter() {parameterName=ColorTint_Evening, defaultValue=<rgba 0.7969 0.4524 0.6187 1>}
  n17 = VectorParameter() {parameterName=ColorTint_Night, defaultValue=<rgba 1 0.818 0.7214 1>}
  MaterialFunctionCall(Emissive=Multiply(
    A=LinearInterpolate(A=@n1.RGB, B=@n2.RGB, Alpha=@n13)
    B=MaterialFunctionCall(Color_Morning=@n14.RGB, Color_Day=@n15.RGB, Color_Evening=@n16.RGB, Color_Night=@n17.RGB) {materialFunction=&Color_Time_of_Day}.Result
   )) {materialFunction=&MF_FX_EmissiveAdjustments}
  Multiply(A=@n12, B=@n2.A)
  n18 = Constant() {r=0.5}
  SphereMask(A=TextureCoordinate(), B=@n18, Radius=@n18) {attenuationRadius=0.7, hardnessPercent=0}
  VectorParameter() {parameterName=BaseColor, defaultValue=<rgba 0.02 0.02 0.02 0>}
 ]
}
