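# NS_Cam_Blink is a Niagara effect in the HeatRay FX set that produces a brief colour flash, apparently to blink the camera view.
# One emitter burst-spawns a very large sprite and writes dynamic material parameters at spawn.
# The active ScaleColor curve steps from dim through a red-heavy spike that eases toward neutral, while alpha fades out over the particle's life.
# An alternative, brighter ScaleColor flash is present but disabled.
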NiagaraSystem /TargetedBeam/Assets/FX/HeatRay/NS_Cam_Blink
dependencies {
 usedDataInterfaces [&NiagaraDataInterfaceColorCurve, &NiagaraDataInterfaceVectorCurve]
 usedDynamicInputs [&VectorFromCurve]
 usedModules [&SystemState]
}
emitters {Minimal {
  emitterUpdateScript [
   &EmitterState()
   &SpawnBurst_Instantaneous(Use Spawn Probability=false, UseLoopCountLimit=false)
  ]
  particleSpawnScript [
   &InitializeParticle(Sprite Size Mode=::NewEnumerator3, Uniform Sprite Size=10000, UsePositionOffset=false, Write Material Random=true)
   &DynamicMaterialParameters(
    Index 0 Param 1=1
    Index 0 Param 2=2
    Module.Index0Float0Write=true
    Module.Index0Float1Write=true
    Module.Index0Float2Write=true
    Module.Index0Float3Write=true
   )
  ]
  particleUpdateScript [
   &ParticleState()
   &ScaleColor(
    Linear Color Curve=DataInterface() {propertyValues {
      AlphaCurve {keys [<at 0 1>,<at 1 0>]}
      BlueCurve {keys [<at 0 1>,<at 0.0247 5.3836>,<at 0.1406 0>,<at 0.4012 1.5>,<at 1 1>]}
      GreenCurve {keys [<at 0 1>,<at 0.0247 9.4516>,<at 0.1406 0>,<at 0.4012 1.1556>,<at 1 1>]}
      RedCurve {keys [<at 0 1>,<at 0.0247 20>,<at 0.1406 0>,<at 0.4012 1.0025>,<at 1 1>]}
     }}
    Scale Mode=::NewEnumerator2
   ) {enabled=false}
   &ScaleColor(
    Linear Color Curve=DataInterface() {propertyValues {
      AlphaCurve {keys [<at 0 1>,<at 1 0>]}
      BlueCurve {keys [<at 0.0006 0.7292 RCIM_Constant>,<at 0.0637 0.9453 RCIM_Constant>,<at 0.1381 1 RCIM_Constant>,<at 1 1>]}
      GreenCurve {keys [<at 0.0006 0.7292 RCIM_Constant>,<at 0.0637 1.241 RCIM_Constant>,<at 0.1381 1 RCIM_Constant>,<at 1 1>]}
      RedCurve {keys [
        <at 0.0006 0.7292 RCIM_Constant>
        <at 0.0637 1.5 RCIM_Constant>
        <at 0.1 2.5042 RCIM_Constant>
        {interpMode=RCIM_Cubic, tangentMode=RCTM_User, time=0.1879, value=1.6381}
        {interpMode=RCIM_Cubic, tangentMode=RCTM_User, time=0.3, value=1.2086}
        <at 1 1>
       ]}
     }}
    Scale Mode=::NewEnumerator2
   ) {module=ScaleColor001}
  ]
  rendererClasses [&NiagaraSpriteRendererProperties]
  renderers [{rendererClass=&NiagaraSpriteRendererProperties}]
 }}
summary {
 systemName=NS_Cam_Blink
 emitters [{emitterName=Minimal, rendererClasses [&NiagaraSpriteRendererProperties]}]
 userVariables []
}
