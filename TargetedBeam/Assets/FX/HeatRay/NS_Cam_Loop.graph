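# Niagara system for the heat ray's looping camera effect, built from huge sprites that persist past their lifetime.
# Screen bursts two red-tinted sprites through four sprite renderers and feeds engine time into dynamic material parameters each update; its flash color curves are disabled.
# Minimal spawns a pale sprite that loops its short lifetime, with a subtle brightness pulse sampled at a randomly offset normalized age.
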NiagaraSystem /TargetedBeam/Assets/FX/HeatRay/NS_Cam_Loop
dependencies {
 usedDataInterfaces [&NiagaraDataInterfaceColorCurve, &NiagaraDataInterfaceVectorCurve]
 usedDynamicInputs [&VectorFromCurve]
 usedModules [&SystemState]
}
emitters {
 Minimal {
  emitterUpdateScript [
   &EmitterState(
    Inactive Response=::NewEnumerator1
    Life Cycle Mode=::NewEnumerator1
    Loop Behavior=::NewEnumerator1
    Loop Duration Mode=::NewEnumerator1
    UseLoopDelay=false
   )
   &SpawnBurst_Instantaneous(Use Spawn Probability=false, UseLoopCountLimit=false)
  ]
  particleSpawnScript [
   &InitializeParticle(
    Color=<rgba 1.2 0.9 0.9 1>
    Lifetime=0.2
    Sprite Size Mode=::NewEnumerator3
    Uniform Sprite Size=10000
    UsePositionOffset=false
    Write Material Random=true
   )
   &DynamicMaterialParameters(
    Index 0 Param 1=1
    Index 0 Param 2=2
    Module.Index0Float0Write=true
    Module.Index0Float1Write=true
    Module.Index0Float2Write=true
    Module.Index0Float3Write=true
   )
  ]
  particleUpdateScript [
   &ParticleState(Kill Particles When Lifetime Has Elapsed=false, Loop Particles Lifetime=true)
   &ScaleColor(
    Linear Color Curve=DataInterface() {propertyValues {
      AlphaCurve {keys [<at 0 1>,<at 1 0>]}
      BlueCurve {keys [<at 0 1>,<at 0.0247 5.3836>,<at 0.1406 0>,<at 0.4012 1.5>,<at 1 1>]}
      GreenCurve {keys [<at 0 1>,<at 0.0247 9.4516>,<at 0.1406 0>,<at 0.4012 1.1556>,<at 1 1>]}
      RedCurve {keys [<at 0 1>,<at 0.0247 20>,<at 0.1406 0>,<at 0.4012 1.0025>,<at 1 1>]}
     }}
    Scale Mode=::NewEnumerator2
   ) {enabled=false}
   &ScaleColor(
    Curve Index=&Add_Float(
     A=$Particles.NormalizedAge
     B=&RandomRangeFloat(Minimum=-0.25, Maximum=0.25, Evaluation Type=::NewEnumerator1)
    )
    Linear Color Curve=DataInterface() {propertyValues {
      AlphaCurve {defaultValue=1, keys []}
      BlueCurve {defaultValue=0.7292, keys [<at 0.0126 1>,<at 0.4786 0.9>,<at 0.9853 1>]}
      GreenCurve {defaultValue=0.7292, keys [<at 0.0126 1>,<at 0.4786 0.9>,<at 0.9853 1>]}
      RedCurve {defaultValue=0.7292, keys [<at 0.0126 1>,<at 0.4786 0.9>,<at 0.9853 1>]}
     }}
    Scale Mode=::NewEnumerator2
   ) {module=ScaleColor001}
  ]
  rendererClasses [&NiagaraSpriteRendererProperties]
  renderers [{rendererClass=&NiagaraSpriteRendererProperties}]
 }
 Screen {
  emitterUpdateScript [
   &EmitterState(
    Inactive Response=::NewEnumerator1
    Life Cycle Mode=::NewEnumerator1
    Loop Behavior=::NewEnumerator1
    Loop Duration Mode=::NewEnumerator1
    UseLoopDelay=false
   )
   &SpawnBurst_Instantaneous(Spawn Count=2, Use Spawn Probability=false, UseLoopCountLimit=false)
  ]
  particleSpawnScript [
   &InitializeParticle(
    Color=<rgba 5 0.9904 0.8333 1>
    Sprite Size Mode=::NewEnumerator3
    Uniform Sprite Size=10000
    UsePositionOffset=false
    Write Material Random=true
   )
   &DynamicMaterialParameters(
    Index 0 Param 1=1
    Index 0 Param 2=2
    Module.Index0Float0Write=true
    Module.Index0Float1Write=true
    Module.Index0Float2Write=true
    Module.Index0Float3Write=true
   )
  ]
  particleUpdateScript [
   &ParticleState(Kill Particles When Lifetime Has Elapsed=false)
   &ScaleColor(
    Linear Color Curve=DataInterface() {propertyValues {
      AlphaCurve {keys [<at 0 1>,<at 1 0>]}
      BlueCurve {keys [<at 0 1>,<at 0.0247 5.3836>,<at 0.1406 0>,<at 0.4012 1.5>,<at 1 1>]}
      GreenCurve {keys [<at 0 1>,<at 0.0247 9.4516>,<at 0.1406 0>,<at 0.4012 1.1556>,<at 1 1>]}
      RedCurve {keys [<at 0 1>,<at 0.0247 20>,<at 0.1406 0>,<at 0.4012 1.0025>,<at 1 1>]}
     }}
    Scale Mode=::NewEnumerator2
   ) {enabled=false}
   &ScaleColor(
    Linear Color Curve=DataInterface() {propertyValues {
      AlphaCurve {keys [<at 0 1>,<at 1 0>]}
      BlueCurve {keys [<at 0.0006 0.4996 RCIM_Constant>,<at 0.0637 2.8607 RCIM_Constant>,<at 0.1381 1 RCIM_Constant>,<at 1 1>]}
      GreenCurve {keys [<at 0.0006 0.4996 RCIM_Constant>,<at 0.0637 4.0011 RCIM_Constant>,<at 0.1381 1 RCIM_Constant>,<at 1 1>]}
      RedCurve {keys [
        <at 0.0006 0.4996 RCIM_Constant>
        <at 0.0637 5 RCIM_Constant>
        <at 0.1369 5 RCIM_Constant>
        {arriveTangent=-4.6343, interpMode=RCIM_Cubic, leaveTangent=-4.6343, time=0.1879, value=3}
        <at 1 1>
       ]}
     }}
    Scale Mode=::NewEnumerator2
   ) {enabled=false, module=ScaleColor001}
   &DynamicMaterialParameters(
    Index 0 Param 1=$Engine.Time
    Index 0 Param 2=0.2744
    Index 0 Param 4=-0.01
    Module.Index0Float0Write=true
    Module.Index0Float1Write=true
    Module.Index0Float2Write=true
    Module.Index0Float3Write=true
   ) {module=DynamicMaterialParameters001}
  ]
  rendererClasses [&NiagaraSpriteRendererProperties]
  renderers [
   {rendererClass=&NiagaraSpriteRendererProperties}
   {rendererClass=&NiagaraSpriteRendererProperties, rendererIndex=1}
   {rendererClass=&NiagaraSpriteRendererProperties, rendererIndex=2}
   {rendererClass=&NiagaraSpriteRendererProperties, rendererIndex=3}
  ]
 }
}
summary {
 systemName=NS_Cam_Loop
 emitters [
  {emitterName=Screen, rendererClasses [&NiagaraSpriteRendererProperties]}
  {emitterName=Minimal, rendererClasses [&NiagaraSpriteRendererProperties]}
 ]
 userVariables []
}
