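# Niagara system for the build-up (charge) phase of the IcicleLarva heat ray.
# Ray draws a faint red sprite beam between system StartPos and EndPos, jittering in width and brightening late over buildUpDuration.
# GPU Sparks spawn around StartPos and are lerped inward rather than force-driven, with velocity set by FindVel for speed-based stretching.
# Start_Flare adds two flickering sprites that grow across the build-up.
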NiagaraSystem /TargetedBeam/Assets/FX/HeatRay/NS_IcicleLarva_HeatRay_BuildUp
dependencies {
 usedDataInterfaces [&NiagaraDataInterfaceCurve, &NiagaraDataInterfaceArrayFloat2, &NiagaraDataInterfaceVectorCurve]
 usedDynamicInputs [
  &UniformAOrBFloat
  &MakeCustomVectorFromBool
  &SetBoolByFloatComparison
  &RandomVector
  &VectorLength
  &Subtract_Vector
  &FindQuatBetweenTwoVectors
  &VectorFromCurve
  &Frac_Float
 ]
 usedModules [
  &self:SystemSpawnScript.SystemScriptSource.SystemScriptGraph.NiagaraNodeAssignment_2.SetVariables_#1
  &SystemState
  &self:SystemSpawnScript.SystemScriptSource.SystemScriptGraph.NiagaraNodeAssignment_3.SetVariables_#2
  &self:FIndSolidStartPos
  &self:SystemSpawnScript.SystemScriptSource.SystemScriptGraph.NiagaraNodeAssignment_0.SetVariables_#3
  &self:SystemSpawnScript.SystemScriptSource.SystemScriptGraph.NiagaraNodeAssignment_1.SetVariables_#4
  &self:SystemSpawnScript.SystemScriptSource.SystemScriptGraph.NiagaraNodeAssignment_4.SetVariables_#5
  &self:SystemSpawnScript.SystemScriptSource.SystemScriptGraph.NiagaraNodeAssignment_5.SetVariables_#6
 ]
}
emitters {
 Ray {
  emitterUpdateScript [
   &EmitterState(Life Cycle Mode=::NewEnumerator1, Loop Behavior=::NewEnumerator1, Loop Duration=1, UseLoopDelay=false)
   &SpawnBurst_Instantaneous(Use Spawn Probability=false, UseLoopCountLimit=false)
  ]
  particleSpawnScript [&InitializeParticle(Color=<rgba 0.4948 0.0206 0.0206 0.2957>, Lifetime=$System.buildUpDuration, UsePositionOffset=false, Write Material Random=true)]
  particleUpdateScript [
   &ParticleState()
   &self:SpriteBeam(
    End:NiagaraPosition=$System.EndPos
    Start:NiagaraPosition=$System.StartPos
    Width:NiagaraFloat=&Multiply_Float(
     A=&FloatFromCurve(FloatCurve=DataInterface() {propertyValues {Curve {
         defaultValue=1
         keys [
          {tangentMode=RCTM_User, time=0, value=0}
          {arriveTangent=1.1111, interpMode=RCIM_Cubic, leaveTangent=1.1111, time=0.1, value=0.8}
          <at 0.9 1>
          <at 1 0 RCIM_Cubic>
         ]
        }}})
     B=&RandomRangeFloat(Minimum=20, Maximum=100, Evaluation Type=::NewEnumerator1)
    )
   ) {module=SpriteBeam}
   &DynamicMaterialParameters(
    Index 0 Param 1=0.1
    Module.Index0Float0Write=true
    Module.Index0Float1Write=true
    Module.Index0Float2Write=true
    Module.Index0Float3Write=true
   )
   &ScaleColor(
    Scale Mode=::NewEnumerator1
    Scale RGBA=&MakeVector4_FromFloat(Float=&MakeFloatFromVector(
      Vector=&VectorFromFloat(Value=&FloatFromCurve(FloatCurve=DataInterface() {propertyValues {Curve {
           defaultValue=0
           keys [
            {arriveTangent=0.2069, interpMode=RCIM_Cubic, leaveTangent=0.2069, tangentMode=RCTM_User, time=0.4, value=0}
            {arriveTangent=0.5, interpMode=RCIM_Cubic, leaveTangent=0.5, time=0.5, value=0.1}
            {arriveTangent=1.1179, interpMode=RCIM_Cubic, leaveTangent=1.1179, tangentMode=RCTM_User, time=0.8, value=0.2}
            {arriveTangent=0.0224, interpMode=RCIM_Cubic, leaveTangent=0.0224, tangentMode=RCTM_User, time=0.9, value=1}
           ]
          }}}))
      Channel=::NewEnumerator0
     ))
    ScaleRGBA=true
   )
  ]
  rendererClasses [&NiagaraSpriteRendererProperties]
  renderers [{rendererClass=&NiagaraSpriteRendererProperties}]
 }
 Sparks {
  simTarget=GPUComputeSim
  emitterUpdateScript [
   &EmitterState(Life Cycle Mode=::NewEnumerator1, Loop Behavior=::NewEnumerator1, Loop Delay=0.15, Loop Duration=1, UseLoopDelay=false)
   &SpawnRate(
    SpawnRate=&FloatFromCurve(
     FloatCurve=DataInterface() {propertyValues {Curve {
        defaultValue=1
        keys [
         <at 0 0.125 RCIM_Cubic>
         {arriveTangent=1.925, interpMode=RCIM_Cubic, leaveTangent=1.925, time=0.3, value=0.375}
         {arriveTangent=2.0515, interpMode=RCIM_Cubic, leaveTangent=-4.3144, tangentMode=RCTM_Break, time=0.4545, value=1}
         {arriveTangent=-2.8947, interpMode=RCIM_Cubic, leaveTangent=-2.8947, time=0.5392, value=0.25}
         <at 0.8 0>
        ]
       }}}
     CurveIndex=$Emitter.NormalizedLoopAge
     Scale Curve=60
    )
    Use Spawn Probability=false
   )
   &SpawnBurst_Instantaneous(
    Spawn Count=&RandomRangeInt(Minimum=10, Maximum=50, Evaluation Type=::NewEnumerator1)
    Use Spawn Probability=false
    UseLoopCountLimit=false
   ) {enabled=false}
  ]
  particleSpawnScript [
   &InitializeParticle(
    Color=<rgba 4 0.2292 0.2292 2>
    Lifetime=&FloatFromCurve(FloatCurve=DataInterface() {propertyValues {Curve {keys [<at 0 1>,<at 1.5 0.25>]}}}, CurveIndex=$Emitter.Age)
    Sprite Size Mode=::NewEnumerator3
    Uniform Sprite Size=&FloatFromCurve(
     FloatCurve=DataInterface() {propertyValues {Curve {keys [
         {arriveTangent=-2.2424, interpMode=RCIM_Cubic, leaveTangent=-2.2424, tangentMode=RCTM_User, time=0, value=1.5}
         {arriveTangent=-1.2, interpMode=RCIM_Cubic, leaveTangent=-1.2, time=0.2, value=0.55}
         <at 1 0.3 RCIM_Cubic>
        ]}}}
     CurveIndex=&RandomRangeFloat(Minimum=0, Maximum=1)
    )
    UsePositionOffset=false
    Write Material Random=true
   )
   &ShapeLocation(
    Hemisphere Distribution=<1 0.5>
    MaskSpawnGroup=false
    Override Seed=false
    Shape Origin=&TransformPosition(Position=$System.StartPos, Source Space=::World, Destination Space=::Simulation, Apply Scale=true)
    Sphere Radius=&RandomRangeFloat(Minimum=50, Maximum=200)
    Sphere Surface Distribution=1
    UseOwnerScale=true
   )
   &AddVelocity(
    Module.Use Falloff Distance=false
    Override Seed=false
    Velocity Mode=::NewEnumerator1
    Velocity Speed=&RandomRangeFloat(Minimum=500, Maximum=2000)
   ) {enabled=false}
   /SetVariables(Particles.Position=&AddVectorToPosition(Position=$Particles.Position, Vector=<0 0 464.0114>)) {enabled=false, isSetParameters=true}
  ]
  particleUpdateScript [
   &ParticleState()
   &CurlNoiseForce(
    Bias Noise=false
    Noise Frequency=5
    Noise Strength=&FloatFromCurve(
     FloatCurve=DataInterface() {propertyValues {Curve {
        defaultValue=1
        keys [
         {arriveTangent=-3.2557, interpMode=RCIM_Cubic, leaveTangent=-3.2557, tangentMode=RCTM_User, time=0, value=1}
         {arriveTangent=-0.7893, interpMode=RCIM_Cubic, leaveTangent=-0.7893, tangentMode=RCTM_User, time=0.1, value=0.25}
         {interpMode=RCIM_Cubic, leaveTangent=-8.0089, time=1, value=0}
        ]
       }}}
     Scale Curve=500
    )
    Pan Noise=false
   ) {enabled=false}
   &Drag(
    Drag=&FloatFromCurve(
     FloatCurve=DataInterface() {propertyValues {Curve {defaultValue=1, keys [{tangentMode=RCTM_User, time=0, value=1}, <at 0.1 0.05>, <at 0.3 0.025>, {tangentMode=RCTM_User, time=1, value=0}]}}}
     Scale Curve=20
    )
    Use Linear Drag=true
    Use Rotational Drag=true
   ) {enabled=false}
   &PointForce(Force Strength=-200, Module.Use Falloff=false, Module.UseLinearForce=false) {enabled=false}
   &GravityForce() {enabled=false}
   /SetVariables(Particles.Position=&LerpPosition(
     A=&TransformPosition(Position=$System.StartPos, Source Space=::World, Destination Space=::Simulation, Apply Scale=true)
     B=$Particles.Initial.Position
     Alpha=&FloatFromCurve(FloatCurve=DataInterface() {propertyValues {Curve {keys [
          {arriveTangent=0.0366, interpMode=RCIM_Cubic, leaveTangentWeight=0.3333, time=0, value=1}
          {arriveTangent=-2.2395, interpMode=RCIM_Cubic, leaveTangent=-2.2395, time=0.1962, value=0.923}
          {arriveTangent=-0.5997, interpMode=RCIM_Cubic, leaveTangent=-0.5997, tangentMode=RCTM_User, time=0.3333, value=0.2536}
          {leaveTangent=-0.1569, time=1.0024, value=0.0425}
         ]}}})
    )) {isSetParameters=true}
   &SolveForcesAndVelocity(Clamp Velocity=false, Limit Acceleration=false) {enabled=false}
   /SetVariables(Particles.Velocity=&self:FindVel(float:NiagaraFloat=2)) {isSetParameters=true}
   &ScaleSpriteSizeBySpeed(Max Scale Factor=<1 80>, Velocity Threshold=2000)
   &ScaleSpriteSize(Uniform Curve Sprite Scale=DataInterface() {propertyValues {Curve {defaultValue=0, keys [{tangentMode=RCTM_User, time=0.9, value=1}, {tangentMode=RCTM_User, time=1, value=0}]}}})
   &ScaleColor(
    Scale Alpha=&FloatFromCurve(FloatCurve=DataInterface() {propertyValues {Curve {keys [<at 0 1>,<at 1 0>]}}})
    ScaleA=true
    ScaleRGB=true
   ) {enabled=false}
  ]
  rendererClasses [&NiagaraSpriteRendererProperties]
  renderers [{rendererClass=&NiagaraSpriteRendererProperties}]
 }
 Start_Flare {
  emitterUpdateScript [
   &EmitterState(Life Cycle Mode=::NewEnumerator1, Loop Behavior=::NewEnumerator1, Loop Duration=1, UseLoopDelay=false)
   &SpawnBurst_Instantaneous(Spawn Count=2, Spawn Time=0.1, Use Spawn Probability=false, UseLoopCountLimit=false)
  ]
  particleSpawnScript [&InitializeParticle(
    Color=<rgba 0.6406 0.0133 0.0266 0.15>
    Lifetime=$System.buildUpDuration
    Position Mode=::NewEnumerator1
    Sprite Size=&Multiply_Vector2D_ByFloat(
     Float=50
     Vector2D=&SelectVector2DFromArray(Direct Array Index=$Engine.ExecIndex, Vector 2D Selection Array=DataInterface() {propertyValues {FloatData [<12 0.1>, <8 1.5>]}})
    )
    Sprite Size Mode=::NewEnumerator4
    UsePositionOffset=false
    Write Material Random=true
   )]
  particleUpdateScript [
   &ParticleState(Kill Particles When Lifetime Has Elapsed=false)
   &ScaleSpriteSize(Uniform Curve Sprite Scale=DataInterface() {propertyValues {Curve {keys [
        {arriveTangent=0.625, interpMode=RCIM_Cubic, time=0.2, value=0.125}
        {arriveTangent=2.5774, interpMode=RCIM_Cubic, leaveTangent=2.5774, tangentMode=RCTM_User, time=1, value=1}
       ]}}}) {module=ScaleSpriteSize001}
   &ScaleSpriteSize(
    Scale Sprite Size Mode=::NewEnumerator0
    Uniform Scale Factor=&RandomRangeFloat(Minimum=0.5, Maximum=1, Evaluation Type=::NewEnumerator1)
   )
  ]
  rendererClasses [&NiagaraSpriteRendererProperties]
  renderers [{rendererClass=&NiagaraSpriteRendererProperties}]
 }
}
summary {
 systemName=NS_IcicleLarva_HeatRay_BuildUp
 emitters [
  {emitterName=Ray, rendererClasses [&NiagaraSpriteRendererProperties]}
  {emitterName=Start_Flare, rendererClasses [&NiagaraSpriteRendererProperties]}
  {emitterName=Sparks, simTarget=GPUComputeSim, rendererClasses [&NiagaraSpriteRendererProperties]}
 ]
 userVariables [
  User.BeamEnd_0 {type=NiagaraPosition, description="", defaultValue {}}
  User.BeamStart_0_0 {type=NiagaraPosition, description="", defaultValue {}}
  User.bIsFrontEnd {type=NiagaraBool}
  User.bIsFrontEndPreview {type=NiagaraBool}
  User.ParaGlideLeanAlpha {type=NiagaraFloat, defaultValue=0, description=""}
 ]
}
