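# Niagara system for a short, one-shot red heat ray, named for the Icicle Larva, with a beam and a start flare.
# RayBurst stretches sprites between system start and end positions via a SpriteBeam module, offsetting the start along the owner's X axis by an indexed eye value.
# Beam width flickers randomly under a curve while brightness drops sharply early, then fades.
# Start_Flare001 bursts two size-jittering sprites, a wide streak and a square glow.
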NiagaraSystem /TargetedBeam/Assets/FX/HeatRay/NS_IcicleLarva_HeatRay_Once
dependencies {
 usedDataInterfaces [
  &NiagaraDataInterfaceArrayFloat3
  &NiagaraDataInterfaceArrayFloat2
  &NiagaraDataInterfaceCurve
  &NiagaraDataInterfaceArrayFloat
  &NiagaraDataInterfaceVectorCurve
 ]
 usedDynamicInputs [
  &SelectVectorFromArray
  &VectorLength
  &Subtract_Vector
  &FindQuatBetweenTwoVectors
  &VectorFromCurve
  &Frac_Float
 ]
 usedModules [
  &self:SystemSpawnScript.SystemScriptSource.SystemScriptGraph.NiagaraNodeAssignment_2.SetVariables_#1
  &SystemState
  &self:SystemSpawnScript.SystemScriptSource.SystemScriptGraph.NiagaraNodeAssignment_3.SetVariables_#2
  &self:FIndSolidStartPos
  &self:SystemSpawnScript.SystemScriptSource.SystemScriptGraph.NiagaraNodeAssignment_0.SetVariables_#3
  &self:SystemSpawnScript.SystemScriptSource.SystemScriptGraph.NiagaraNodeAssignment_1.SetVariables_#4
  &self:SystemSpawnScript.SystemScriptSource.SystemScriptGraph.NiagaraNodeAssignment_4.SetVariables_#5
 ]
}
emitters {
 RayBurst {
  emitterUpdateScript [
   &EmitterState(Life Cycle Mode=::NewEnumerator1, Loop Behavior=::NewEnumerator1, Loop Duration=1, UseLoopDelay=false)
   &SpawnBurst_Instantaneous(Use Spawn Probability=false, UseLoopCountLimit=false)
  ]
  particleSpawnScript [
   &InitializeParticle(Color=<rgba 0.276 0.0115 0.0115 0.075>, Lifetime=0.25, UsePositionOffset=false, Write Material Random=true)
   /SetVariables(Particles.EyeIndex:NiagaraFloat=&SelectFloatFromArray(Direct Array Index=$Engine.ExecIndex, Float Selection Array=DataInterface() {propertyValues {FloatData [5,-5]}})) {isSetParameters=true}
  ]
  particleUpdateScript [
   /SetVariables(Particles.EyeOffset:Vector3f=&TransformVector(
     Vector=&Multiply_VectorByFloat(Vector=$Engine.Owner.SystemXAxis, Float=$Particles.EyeIndex)
     Destination Space=::Simulation
    )) {isSetParameters=true}
   &ParticleState()
   &self:SpriteBeam_01(
    End:NiagaraPosition=$System.EndPos
    Start:NiagaraPosition=&AddVectorToPosition(Position=$System.StartPos, Vector=$Particles.EyeOffset)
    Width:NiagaraFloat=&Multiply_Float(
     A=&FloatFromCurve(FloatCurve=DataInterface() {propertyValues {Curve {
         defaultValue=1
         keys [
          {tangentMode=RCTM_User, time=0, value=0}
          {arriveTangent=1.1111, interpMode=RCIM_Cubic, leaveTangent=1.1111, time=0.1, value=0.8}
          <at 0.9 1>
          <at 1 0 RCIM_Cubic>
         ]
        }}})
     B=&RandomRangeFloat(Minimum=50, Maximum=100, Evaluation Type=::NewEnumerator1)
    )
   ) {module=SpriteBeam_01}
   &DynamicMaterialParameters(
    Index 0 Param 1=1
    Module.Index0Float0Write=true
    Module.Index0Float1Write=true
    Module.Index0Float2Write=true
    Module.Index0Float3Write=true
   )
   &ScaleColor(
    Scale Mode=::NewEnumerator1
    Scale RGBA=/Niagara/DynamicInputs/Multiply/Multiply_Vector4_ByFloat(
     Vector4:Vector4f=<1 1 1 1>
     Float:NiagaraFloat=&FloatFromCurve(
      FloatCurve=DataInterface() {propertyValues {Curve {keys [
          {arriveTangent=-0.0076, interpMode=RCIM_Cubic, leaveTangent=-0.0076, tangentMode=RCTM_User, time=0.1879, value=1}
          {arriveTangent=-0.5753, interpMode=RCIM_Cubic, leaveTangent=-0.5753, tangentMode=RCTM_User, time=0.2005, value=0.15}
          <at 1 0 RCIM_Cubic>
         ]}}}
      Scale Curve=10
     )
    )
    ScaleRGBA=true
   )
  ]
  rendererClasses [&NiagaraSpriteRendererProperties]
  renderers [{rendererClass=&NiagaraSpriteRendererProperties}]
 }
 Start_Flare001 {
  emitterUpdateScript [
   &EmitterState(Life Cycle Mode=::NewEnumerator1, Loop Behavior=::NewEnumerator1, Loop Duration=1, UseLoopDelay=false)
   &SpawnBurst_Instantaneous(Spawn Count=2, Use Spawn Probability=false, UseLoopCountLimit=false)
  ]
  particleSpawnScript [
   &InitializeParticle(
    Color=<rgba 2 0.0417 0.0831 0.3>
    Lifetime=0.2
    Sprite Size=&Multiply_Vector2D_ByFloat(
     Float=200
     Vector2D=&SelectVector2DFromArray(Direct Array Index=$Engine.ExecIndex, Vector 2D Selection Array=DataInterface() {propertyValues {FloatData [<4 0.1>, <1 1>]}})
    )
    Sprite Size Mode=::NewEnumerator4
    UsePositionOffset=false
    Write Material Random=true
   )
   /SetVariables(Particles.SpriteSize001:Vector2f=&Multiply_Vector2D(A=$Particles.SpriteSize, B=<1 10>)) {enabled=false, isSetParameters=true}
   &CameraOffset(Camera Offset Amount=10) {enabled=false}
  ]
  particleUpdateScript [
   &ParticleState()
   &ScaleSpriteSize(
    Scale Sprite Size Mode=::NewEnumerator0
    Uniform Scale Factor=&RandomRangeFloat(Minimum=0.5, Maximum=1, Evaluation Type=::NewEnumerator1)
   )
  ]
  rendererClasses [&NiagaraSpriteRendererProperties]
  renderers [
   {rendererClass=&NiagaraSpriteRendererProperties}
   {rendererClass=&NiagaraSpriteRendererProperties, rendererIndex=1}
  ]
 }
}
summary {
 systemName=NS_IcicleLarva_HeatRay_Once
 emitters [
  {emitterName=Start_Flare001, rendererClasses [&NiagaraSpriteRendererProperties]}
  {emitterName=RayBurst, rendererClasses [&NiagaraSpriteRendererProperties]}
 ]
 userVariables [
  User.BeamEnd_0 {type=NiagaraPosition, description="", defaultValue {}}
  User.BeamStart_0_0 {type=NiagaraPosition, description="", defaultValue {}}
  User.bIsFrontEnd {type=NiagaraBool}
  User.bIsFrontEndPreview {type=NiagaraBool}
  User.ParaGlideLeanAlpha {type=NiagaraFloat, defaultValue=0, description=""}
 ]
}
