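# Niagara system providing the light for the IcicleLarva heat ray's impact effect, with a single burst-spawned emitter using light and component renderers.
# Each update sets light radius from a random range scaled up with owner-to-camera distance, or zero when System.bCompleteOnInactive is false.
# Position combines random jitter, which widens with camera distance, with a fixed-length offset along a camera vector.
# Particles are not killed when their lifetime elapses, and volumetric scattering is enabled.
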NiagaraSystem /TargetedBeam/Assets/FX/HeatRay/NS_IcicleLarva_RayImpact_Light
dependencies {usedDataInterfaces [&NiagaraDataInterfaceCamera], usedDynamicInputs [], usedModules [&SystemState]}
emitters {Light {
  emitterUpdateScript [
   &EmitterState(
    Inactive Response=::NewEnumerator1
    Life Cycle Mode=::NewEnumerator1
    Loop Behavior=::NewEnumerator1
    Loop Duration Mode=::NewEnumerator1
    UseLoopDelay=false
   )
   &SpawnBurst_Instantaneous(Use Spawn Probability=false, UseLoopCountLimit=false)
  ]
  particleSpawnScript [&InitializeParticle(
    Color Mode=::NewEnumerator0
    Lifetime=5
    Sprite Size=<5 1000>
    Sprite Size Mode=::NewEnumerator4
    UsePositionOffset=false
    Write Material Random=true
   )]
  particleUpdateScript [
   &ParticleState(Kill Particles When Lifetime Has Elapsed=false, Let Infinitely Lived Particles Die When Emitter Deactivates=true)
   /SetVariables(Particles.LightVolumetricScattering=1) {isSetParameters=true}
   /SetVariables(Particles.LightRadius=&MakeCustomFloatFromBool(
     Bool=$System.bCompleteOnInactive
     True Float=&Multiply_Float(
      A=&RemapRange(
       Input Value To Remap=/Niagara/DynamicInputs/Camera/DistanceToCamera(Use Owner Position:NiagaraBool=true, Camera Query:NiagaraDataInterfaceCamera=DataInterface())
       Input Min=500
       Input Max=10000
       Output Min=0.03
       Output Max=1
      )
      B=&RandomRangeFloat(Minimum=3000, Maximum=6000, Evaluation Type=::NewEnumerator1)
     )
     False Float=0
    )) {isSetParameters=true}
   /SetVariables(Particles.Position=&AddVectorToPosition(
     Position=<0 0 0>
     Vector=&Add_Vector(
      A=&RandomVector(
       Vector Scale=&RemapRange(
        Input Value To Remap=&DistanceToCamera(Use Owner Position:NiagaraBool=true, Camera Query:NiagaraDataInterfaceCamera=DataInterface())
        Input Min=1000
        Input Max=10000
        Output Min=1
        Output Max=5
       )
       Evaluation Type=::NewEnumerator1
      )
      B=&Multiply_VectorByFloat(Vector=&ReturnCameraVector(Camera Vectors=::NewEnumerator1, Coordinate Space=::Simulation), Float=-100)
     )
    )) {isSetParameters=true}
  ]
  rendererClasses [&NiagaraComponentRendererProperties, &NiagaraLightRendererProperties]
  renderers [
   {rendererClass=&NiagaraComponentRendererProperties, rendererIndex=0}
   {rendererClass=&NiagaraLightRendererProperties, rendererIndex=1}
  ]
 }}
summary {
 systemName=NS_IcicleLarva_RayImpact_Light
 emitters [{emitterName=Light, rendererClasses [&NiagaraComponentRendererProperties, &NiagaraLightRendererProperties]}]
 userVariables [
  User.bIsMudImpact {type=NiagaraBool}
  User.bIsWaterImpact {type=NiagaraBool}
  User.bLight {type=NiagaraBool, defaultValue=true}
  User.PhysicalSurface {type=NiagaraInt32}
  User.RibbonID {type=NiagaraInt32}
  User.WaterBodyIndex {type=NiagaraFloat, defaultValue=0, description=""}
  User.WaterDepth {type=NiagaraFloat, defaultValue=0, description=""}
  User.WaveOffset {type=Vector3f, description="", defaultValue {struct=Vector3f, value=<0 0 0>}}
 ]
}
